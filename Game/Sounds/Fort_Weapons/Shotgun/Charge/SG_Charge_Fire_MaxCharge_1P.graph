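# First-person sound cue for firing the charge shotgun at maximum charge.
# A random pick of two max-charge shot waves is mixed with a quieter sub-hit layer, then passed through a modulator for boosted volume and slight upward pitch variation.
# The quality-level node sends every tier to the same mixer, so the content is identical at each.
# It uses first-person weapon attenuation, a ranged-fire sound class and a gunfire concurrency set.
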SoundCue /Game/Sounds/Fort_Weapons/Shotgun/Charge/SG_Charge_Fire_MaxCharge_1P
root=Modulator_2
nodes [
 {id=Mixer_0, children [Random_2,WavePlayer_3], properties {inputVolume [1,0.4]}}
 {id=Modulator_2, children [QualityLevel_0], properties {pitchMax=1.1, pitchMin=1, volumeMax=1.75, volumeMin=1.6}}
 {id=QualityLevel_0, children [Mixer_0,Mixer_0,Mixer_0,Mixer_0]}
 {id=Random_2, children [WavePlayer_0,WavePlayer_1], properties {weights [1,1]}}
 {id=WavePlayer_0, properties {soundWaveAssetPtr=&/SG_Charge_Shot_Max_01}}
 {id=WavePlayer_1, properties {soundWaveAssetPtr=&/SG_Charge_Shot_Max_02}}
 {id=WavePlayer_3, properties {soundWaveAssetPtr=&/Sub_Hit_N445}}
]
properties {
 attenuationSettings=&Fort_Weapon_Projectile_FirstPerson_Athena_Att
 bDebug=true
 maxDistance=2097152
 soundClassObject=&Fort_WeaponRanged_Fire_1P_PSM
 volumeMultiplier=0.7
 concurrencyOverrides {bLimitToOwner=true, maxCount=3, resolutionRule=StopOldest}
 concurrencySet [&Weapon_Gunfire_FirstPerson_Conc]
}
