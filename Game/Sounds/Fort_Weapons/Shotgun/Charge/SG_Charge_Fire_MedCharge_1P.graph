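# First-person fire sound cue for the charge shotgun at medium charge.
# A random node picks one of three medium-charge shot waves, excluded from branch culling, and a mixer layers it with a constant sub punch wave.
# A modulator adds slight upward pitch and boosted volume variation on top.
# It uses first-person weapon attenuation, sound class and gunfire concurrency, with owner-limited stop-oldest concurrency overrides also specified.
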SoundCue /Game/Sounds/Fort_Weapons/Shotgun/Charge/SG_Charge_Fire_MedCharge_1P
root=Modulator_2
nodes [
 {id=Mixer_0, children [Random_2,WavePlayer_3], properties {inputVolume [1.25,1]}}
 {id=Modulator_2, children [Mixer_0], properties {pitchMax=1.1, pitchMin=1, volumeMax=1.6, volumeMin=1.5}}
 {id=Random_2, children [WavePlayer_0,WavePlayer_1,WavePlayer_2], properties {bShouldExcludeFromBranchCulling=true, weights [1,1,1]}}
 {id=WavePlayer_0, properties {soundWaveAssetPtr=&/SG_Charge_Shot_Med_01}}
 {id=WavePlayer_1, properties {soundWaveAssetPtr=&/SG_Charge_Shot_Med_02}}
 {id=WavePlayer_2, properties {soundWaveAssetPtr=&/SG_Charge_Shot_Med_03}}
 {id=WavePlayer_3, properties {soundWaveAssetPtr=&/Sub_Punch_01}}
]
properties {
 attenuationSettings=&Fort_Weapon_Projectile_FirstPerson_Athena_Att
 bDebug=true
 maxDistance=2097152
 soundClassObject=&Fort_WeaponRanged_Fire_1P_PSM
 volumeMultiplier=0.7
 concurrencyOverrides {bLimitToOwner=true, maxCount=3, resolutionRule=StopOldest}
 concurrencySet [&Weapon_Gunfire_FirstPerson_Conc]
}
