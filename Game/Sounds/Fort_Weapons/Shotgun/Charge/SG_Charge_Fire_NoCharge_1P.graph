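# First-person sound cue for firing the charge shotgun with no charge.
# A Random node picks one of three equally weighted low-charge shot waves, feeding a Modulator that varies pitch slightly upward at constant volume.
# First-person weapon attenuation, a ranged-fire sound class and a gunfire concurrency set govern playback, alongside owner-limited, stop-oldest override settings.
# The cue has a boosted volume multiplier and its debug flag is on.
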SoundCue /Game/Sounds/Fort_Weapons/Shotgun/Charge/SG_Charge_Fire_NoCharge_1P
root=Modulator_2
nodes [
 {id=Modulator_2, children [Random_2], properties {pitchMax=1.1, pitchMin=1, volumeMax=1, volumeMin=1}}
 {id=Random_2, children [WavePlayer_3,WavePlayer_4,WavePlayer_5], properties {bShouldExcludeFromBranchCulling=true, weights [1,1,1]}}
 {id=WavePlayer_3, properties {soundWaveAssetPtr=&/SG_Charge_Shot_Low_03}}
 {id=WavePlayer_4, properties {soundWaveAssetPtr=&/SG_Charge_Shot_Low_01}}
 {id=WavePlayer_5, properties {soundWaveAssetPtr=&/SG_Charge_Shot_Low_02}}
]
properties {
 attenuationSettings=&Fort_Weapon_Projectile_FirstPerson_Athena_Att
 bDebug=true
 maxDistance=2097152
 soundClassObject=&Fort_WeaponRanged_Fire_1P_PSM
 volumeMultiplier=1.5
 concurrencyOverrides {bLimitToOwner=true, maxCount=3, resolutionRule=StopOldest}
 concurrencySet [&Weapon_Gunfire_FirstPerson_Conc]
}
