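# Sound cue for the low-ammo shot layer of the charge shotgun.
# A single wave player feeds a modulator that randomizes pitch, with volume held constant and the cue's base pitch slightly lowered.
# It routes to the degraded low-ammo ranged weapon sound class and overrides concurrency so new instances are refused once the limit is reached.
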SoundCue /Game/Sounds/Fort_Weapons/Shotgun/Charge/SG_Charge_LowAmmo_Cue
root=Modulator_1
nodes [
 {id=Modulator_1, children [WavePlayer_2], properties {pitchMax=1.1, volumeMax=1, volumeMin=1}}
 {id=WavePlayer_2, properties {soundWaveAssetPtr=&/SG_Charge_Shot_LowAmmo_01}}
]
properties {
 bOverrideConcurrency=true
 maxDistance=2097152
 pitchMultiplier=0.95
 soundClassObject=/Game/Sounds/FortSoundClassesAndModes/Fort_WeaponRanged_LowAmmo_Degraded_PSM
 volumeMultiplier=1
 concurrencyOverrides {resolutionRule=PreventNew}
}
