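# Sound cue for the charged shotgun's aim-down-sights-in foley, classed as weapon foley.
# It reuses a drum shotgun reload-start wave, played quietly and pitched up, through a Modulator that varies each play.
# It uses the weapon reload attenuation and overrides concurrency to prevent new instances once the limit is reached.
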SoundCue /Game/Sounds/Fort_Weapons/Shotgun/Charge/SG_Charged_ADS_In
root=Modulator_0
nodes [
 {id=Modulator_0, children [WavePlayer_1], properties {pitchMax=1.15}}
 {id=WavePlayer_1, properties {soundWaveAssetPtr=/Game/Sounds/Fort_Weapons/Shotgun/Drum/ShotgunDrum_Reload_Start_01}}
]
properties {
 attenuationSettings=&Fort_Weapon_Reload_Att
 bOverrideConcurrency=true
 maxDistance=2250
 pitchMultiplier=1.4
 soundClassObject=&Fort_WeaponFoley
 volumeMultiplier=0.2
 attenuationOverrides {falloffDistance=1600}
 concurrencyOverrides {resolutionRule=PreventNew}
}
