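# Sound cue for the swap sound of the charge shotgun, playing a single release wave.
# The wave runs through a slight random pitch variation, a first-person projectile attenuation node and a sound class node set to the third-person ranged weapon fire class.
# A concurrency override limits it to a few instances per owner, stopping the oldest first.
# The debug flag is left enabled on the cue.
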SoundCue /Game/Sounds/Fort_Weapons/Shotgun/Charge/SG_Charged_Swap
root=SoundClass_2
nodes [
 {
  id=Attenuation_2
  children [Modulator_2]
  properties {attenuationSettings=&Fort_Weapon_Projectile_FirstPerson_Athena_Att}
 }
 {id=Modulator_2, children [WavePlayer_6], properties {pitchMax=1.1, pitchMin=1, volumeMax=1, volumeMin=1}}
 {
  id=SoundClass_2
  children [Attenuation_2]
  properties {soundClassOverride=&Fort_WeaponRanged_Fire_3P_PSM}
 }
 {id=WavePlayer_6, properties {soundWaveAssetPtr=&/SG_Charge_Swap_Release_02}}
]
properties {
 bDebug=true
 bOverrideConcurrency=true
 maxDistance=1800.0001
 soundClassObject=&Fort_WeaponRanged_Fire_3P_PSM
 concurrencyOverrides {bLimitToOwner=true, maxCount=3, resolutionRule=StopOldest}
}
